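# Niagara system for the EdibleCable companion emote, showing balloon sprites, their pops and sparks on the gadget skeletal mesh.
# A change in User.BurstIndex spawns a balloon that tracks its bone, with per-index lifetime, scale-in and renderer visibility tag, and frame from User.Frame.
# Balloon deaths pass through system arrays to spawn short flipbook sprites at those positions.
# Changes in User.BurstSpark trigger falling spark bursts at the BalloonBoss bone.
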NiagaraSystem /CosmeticCompanions/FX/Builtins/EdibleCable/Emote/NS_Emote_EdibleCable_Balloon
dependencies {
 usedDataInterfaces [
  &NiagaraDataInterfaceSkeletalMesh
  &NiagaraDataInterfaceCurve
  &NiagaraDataInterfaceSpriteRendererInfo
  &NiagaraDataInterfaceMeshRendererInfo
  &NiagaraDataInterfaceVector2DCurve
  &NiagaraDataInterfaceArrayPosition
  &NiagaraDataInterfaceArrayColor
 ]
 usedDynamicInputs [
  &MergeHelper_FloatSwitch
  &Multiply_Float
  &MakeCustomFloatFromBool
 ]
 usedModules [
  &self:SystemSpawnScript.SystemScriptSource.SystemScriptGraph.NiagaraNodeAssignment_0.SetVariables_#1
  &SystemState
  &self:ReadDeath_02
  &self:SystemSpawnScript.SystemScriptSource.SystemScriptGraph.NiagaraNodeAssignment_1.SetVariables_#2
 ]
}
emitters {
 NE_Balloon {
  emitterUpdateScript [
   &EmitterState(Life Cycle Mode=::NewEnumerator1, Loop Behavior=::NewEnumerator1, Loop Duration Mode=::NewEnumerator1, UseLoopDelay=false)
   /Game/Effects/Niagara/Modules/Logic/UpdateValueStateChange(InputIntParam:NiagaraInt32=&MakeIntFromFloat(Float=$User.BurstIndex, Conversion Mode=::NewEnumerator1))
   &self:TriggerSpawn_03(
    Spawn Count:NiagaraInt32=&self:Merge_Helper_Int_Switch_15(
     0:NiagaraInt32=1
     1:NiagaraInt32=1
     2:NiagaraInt32=1
     3:NiagaraInt32=1
     4:NiagaraInt32=1
     5:NiagaraInt32=1
     6:NiagaraInt32=1
     7:NiagaraInt32=1
     8:NiagaraInt32=1
     9:NiagaraInt32=1
     10:NiagaraInt32=1
     11:NiagaraInt32=1
     12:NiagaraInt32=1
     13:NiagaraInt32=1
     14:NiagaraInt32=1
     Index:NiagaraInt32=&MakeIntFromFloat(Float=$User.BurstIndex, Conversion Mode=::NewEnumerator1)
     15:NiagaraInt32=1
     Num:NiagaraInt32=14
    )
    Spawn Group:NiagaraInt32=&MakeIntFromFloat(Float=$User.BurstIndex, Conversion Mode=::NewEnumerator1)
    Trigger:NiagaraBool=$Emitter.UpdateValueStateChange.Changed
   ) {module=TriggerSpawn_03}
  ]
  particleSpawnScript [
   /SetVariables(Particles.IndexIn:NiagaraFloat=$User.BurstIndex) {isSetParameters=true}
   /SetVariables(Particles.Lifetimes:NiagaraFloat=&self:Merge_Helper_Float_Switch_15(
     0:NiagaraFloat=1
     1:NiagaraFloat=0.9
     2:NiagaraFloat=0.7
     3:NiagaraFloat=0.35
     4:NiagaraFloat=0.42
     5:NiagaraFloat=0.45
     6:NiagaraFloat=0.5
     7:NiagaraFloat=0.6
     8:NiagaraFloat=0.48
     9:NiagaraFloat=0.5
     10:NiagaraFloat=0.65
     11:NiagaraFloat=0.7
     12:NiagaraFloat=0.8
     13:NiagaraFloat=0.9
     14:NiagaraFloat=0.75
     Index:NiagaraInt32=&SpawnGroup()
     15:NiagaraFloat=2.8
     Num:NiagaraInt32=14
    )) {isSetParameters=true}
   &InitializeParticle(
    Lifetime=$Particles.Lifetimes
    Sprite Size Mode=::NewEnumerator3
    Uniform Sprite Size=31
    UsePositionOffset=false
    Write Material Random=true
   )
  ]
  particleUpdateScript [
   &ParticleState()
   &SkeletalMeshLocation(
    Bone / Socket Index=&MakeIntFromFloat(Float=$Particles.IndexIn, Conversion Mode=::NewEnumerator1)
    Bone Sampling Mode=::NewEnumerator4
    MaskSpawnGroup=false
    Sampled Position Offset=<0 0 -2>
    Skeletal Mesh=DataInterface() {propertyValues {
      DefaultMesh=null
      ExcludeBoneName=root
      OptionalComponentClass=null
      PreviewMesh=/CosmeticCompanions/Assets/Props/Emotes/EdibleCable/Meshes/SKM_EdibleCable_Gadget
      SkinningMode=SkinOnTheFly
      SoftSourceActor=null
      SourceMode=Default
      UvSetIndex=0
      WholeMeshLOD=-1
      bExcludeBone=false
      bReadDeformedGeometry=true
      bRequireCurrentFrameData=true
      ComponentTags []
      FilteredBones [Balloon01,Balloon02,Balloon03,Balloon04,Balloon05,Balloon01,Balloon02,Balloon03,Balloon04,Balloon05,Balloon01,Balloon02,Balloon03,Balloon04,Balloon05,BalloonBoss]
      FilteredSockets []
      MeshUserParameter {parameter {name=None, typeDefHandle {registeredTypeIndex=1}, varData []}}
      SamplingRegions []
     }}
   )
   &self:Sprite_Particles(
    Create SpriteSize Attribute=true
    SpriteSize:Vector2f=&Multiply_Vector2D_ByFloat(
     Float=&FloatFromCurve(FloatCurve=DataInterface() {propertyValues {Curve {defaultValue=1, keys [{tangentMode=RCTM_User, time=0, value=0}, <at 0.085 1>, {tangentMode=RCTM_User, time=1, value=1}]}}})
     Vector2D=<1 1>
    )
   ) {module=Sprite_Particles}
   &self:Sprite_Particles_01(
    Create SpriteSize Attribute=true
    SpawnGroup=1
    SpriteSize:Vector2f=&Multiply_Vector2D_ByFloat(
     Float=&FloatFromCurve(FloatCurve=DataInterface() {propertyValues {Curve {defaultValue=1, keys [{tangentMode=RCTM_User, time=0, value=0}, <at 0.09 1>, {tangentMode=RCTM_User, time=1, value=1}]}}})
     Vector2D=<1 1>
    )
   ) {module=Sprite_Particles_01}
   &self:Sprite_Particles_02(
    Create SpriteSize Attribute=true
    SpawnGroup=2
    SpriteSize:Vector2f=&Multiply_Vector2D_ByFloat(
     Float=&FloatFromCurve(FloatCurve=DataInterface() {propertyValues {Curve {defaultValue=1, keys [{tangentMode=RCTM_User, time=0, value=0}, <at 0.09 1>, {tangentMode=RCTM_User, time=1, value=1}]}}})
     Vector2D=<1 1>
    )
   ) {module=Sprite_Particles_02}
   &self:Sprite_Particles_03(
    Create SpriteSize Attribute=true
    SpawnGroup=3
    SpriteSize:Vector2f=&Multiply_Vector2D_ByFloat(
     Float=&FloatFromCurve(FloatCurve=DataInterface() {propertyValues {Curve {defaultValue=1, keys [{tangentMode=RCTM_User, time=0, value=0}, <at 0.15 1>, {tangentMode=RCTM_User, time=1, value=1}]}}})
     Vector2D=<1 1>
    )
   ) {module=Sprite_Particles_03}
   &self:Sprite_Particles_04(
    Create SpriteSize Attribute=true
    SpawnGroup=4
    SpriteSize:Vector2f=&Multiply_Vector2D_ByFloat(
     Float=&FloatFromCurve(FloatCurve=DataInterface() {propertyValues {Curve {defaultValue=1, keys [{tangentMode=RCTM_User, time=0, value=0}, <at 0.2 1>, {tangentMode=RCTM_User, time=1, value=1}]}}})
     Vector2D=<1 1>
    )
   ) {module=Sprite_Particles_04}
   &self:Sprite_Particles_05(
    Create SpriteSize Attribute=true
    SpawnGroup=5
    SpriteSize:Vector2f=&Multiply_Vector2D_ByFloat(
     Float=&FloatFromCurve(FloatCurve=DataInterface() {propertyValues {Curve {defaultValue=1, keys [{tangentMode=RCTM_User, time=0, value=0}, <at 0.2 1>, {tangentMode=RCTM_User, time=1, value=1}]}}})
     Vector2D=<1 1>
    )
   ) {module=Sprite_Particles_05}
   &self:Sprite_Particles_06(
    Create SpriteSize Attribute=true
    SpawnGroup=6
    SpriteSize:Vector2f=&Multiply_Vector2D_ByFloat(
     Float=&FloatFromCurve(FloatCurve=DataInterface() {propertyValues {Curve {defaultValue=1, keys [{tangentMode=RCTM_User, time=0, value=0}, <at 0.15 1>, {tangentMode=RCTM_User, time=1, value=1}]}}})
     Vector2D=<1 1>
    )
   ) {module=Sprite_Particles_06}
   &self:Sprite_Particles_07(
    Create SpriteSize Attribute=true
    SpawnGroup=7
    SpriteSize:Vector2f=&Multiply_Vector2D_ByFloat(
     Float=&FloatFromCurve(FloatCurve=DataInterface() {propertyValues {Curve {defaultValue=1, keys [{tangentMode=RCTM_User, time=0, value=0}, <at 0.15 1>, {tangentMode=RCTM_User, time=1, value=1}]}}})
     Vector2D=<1 1>
    )
   ) {module=Sprite_Particles_07}
   &self:Sprite_Particles_08(
    Create SpriteSize Attribute=true
    SpawnGroup=8
    SpriteSize:Vector2f=&Multiply_Vector2D_ByFloat(
     Float=&FloatFromCurve(FloatCurve=DataInterface() {propertyValues {Curve {defaultValue=1, keys [{tangentMode=RCTM_User, time=0, value=0}, <at 0.1 1>, {tangentMode=RCTM_User, time=1, value=1}]}}})
     Vector2D=<1 1>
    )
   ) {module=Sprite_Particles_08}
   &self:Sprite_Particles_09(
    Create SpriteSize Attribute=true
    SpawnGroup=9
    SpriteSize:Vector2f=&Multiply_Vector2D_ByFloat(
     Float=&FloatFromCurve(FloatCurve=DataInterface() {propertyValues {Curve {defaultValue=1, keys [{tangentMode=RCTM_User, time=0, value=0}, <at 0.1 1>, {tangentMode=RCTM_User, time=1, value=1}]}}})
     Vector2D=<1 1>
    )
   ) {module=Sprite_Particles_09}
   &self:Sprite_Particles_010(
    Create SpriteSize Attribute=true
    SpawnGroup=10
    SpriteSize:Vector2f=&Multiply_Vector2D_ByFloat(
     Float=&FloatFromCurve(FloatCurve=DataInterface() {propertyValues {Curve {defaultValue=1, keys [{tangentMode=RCTM_User, time=0, value=0}, <at 0.09 1>, {tangentMode=RCTM_User, time=1, value=1}]}}})
     Vector2D=<1 1>
    )
   ) {module=Sprite_Particles_010}
   &self:Sprite_Particles_011(
    Create SpriteSize Attribute=true
    SpawnGroup=11
    SpriteSize:Vector2f=&Multiply_Vector2D_ByFloat(
     Float=&FloatFromCurve(FloatCurve=DataInterface() {propertyValues {Curve {defaultValue=1, keys [{tangentMode=RCTM_User, time=0, value=0}, <at 0.09 1>, {tangentMode=RCTM_User, time=1, value=1}]}}})
     Vector2D=<1 1>
    )
   ) {module=Sprite_Particles_011}
   &self:Sprite_Particles_012(
    Create SpriteSize Attribute=true
    SpawnGroup=12
    SpriteSize:Vector2f=&Multiply_Vector2D_ByFloat(
     Float=&FloatFromCurve(FloatCurve=DataInterface() {propertyValues {Curve {defaultValue=1, keys [{tangentMode=RCTM_User, time=0, value=0}, <at 0.09 1>, {tangentMode=RCTM_User, time=1, value=1}]}}})
     Vector2D=<1 1>
    )
   ) {module=Sprite_Particles_012}
   &self:Sprite_Particles_013(
    Create SpriteSize Attribute=true
    SpawnGroup=13
    SpriteSize:Vector2f=&Multiply_Vector2D_ByFloat(
     Float=&FloatFromCurve(FloatCurve=DataInterface() {propertyValues {Curve {defaultValue=1, keys [{tangentMode=RCTM_User, time=0, value=0}, <at 0.09 1>, {tangentMode=RCTM_User, time=1, value=1}]}}})
     Vector2D=<1 1>
    )
   ) {module=Sprite_Particles_013}
   &self:Sprite_Particles_014(
    Create SpriteSize Attribute=true
    SpawnGroup=14
    SpriteSize:Vector2f=&Multiply_Vector2D_ByFloat(
     Float=&FloatFromCurve(FloatCurve=DataInterface() {propertyValues {Curve {defaultValue=1, keys [{tangentMode=RCTM_User, time=0, value=0}, <at 0.09 1>, {tangentMode=RCTM_User, time=1, value=1}]}}})
     Vector2D=<1 1>
    )
   ) {module=Sprite_Particles_014}
   &self:Sprite_Particles_015(
    Create SpriteSize Attribute=true
    SpawnGroup=15
    SpriteSize:Vector2f=&Multiply_Vector2D_ByFloat(
     Float=&FloatFromCurve(FloatCurve=DataInterface() {propertyValues {Curve {defaultValue=1, keys [{tangentMode=RCTM_User, time=0, value=0}, <at 0.03 1>, {tangentMode=RCTM_User, time=1, value=1}]}}})
     Vector2D=<1 1>
    )
   ) {module=Sprite_Particles_015}
   &ScaleSpriteSize(
    Scale Factor=&self:Merge_Helper_Vector2_Switch_15(
     0:Vector2f=$Particles.Sprite_Particles.SpriteSize
     1:Vector2f=$Particles.Sprite_Particles_01.SpriteSize
     2:Vector2f=$Particles.Sprite_Particles_02.SpriteSize
     3:Vector2f=$Particles.Sprite_Particles_03.SpriteSize
     4:Vector2f=$Particles.Sprite_Particles_04.SpriteSize
     5:Vector2f=$Particles.Sprite_Particles_05.SpriteSize
     6:Vector2f=$Particles.Sprite_Particles_06.SpriteSize
     7:Vector2f=$Particles.Sprite_Particles_07.SpriteSize
     8:Vector2f=$Particles.Sprite_Particles_08.SpriteSize
     9:Vector2f=$Particles.Sprite_Particles_09.SpriteSize
     10:Vector2f=$Particles.Sprite_Particles_010.SpriteSize
     11:Vector2f=$Particles.Sprite_Particles_011.SpriteSize
     12:Vector2f=$Particles.Sprite_Particles_012.SpriteSize
     13:Vector2f=$Particles.Sprite_Particles_013.SpriteSize
     14:Vector2f=$Particles.Sprite_Particles_014.SpriteSize
     Index:NiagaraInt32=&SpawnGroup()
     Num:NiagaraInt32=14
     15:Vector2f=$Particles.Sprite_Particles_015.SpriteSize
    )
    Scale Sprite Size Mode=::NewEnumerator2
   )
   /SetVariables(Particles.VisibilityTag=&self:Merge_Helper_Int_Switch_15(
     0:NiagaraInt32=0
     1:NiagaraInt32=1
     2:NiagaraInt32=2
     3:NiagaraInt32=3
     4:NiagaraInt32=0
     5:NiagaraInt32=1
     6:NiagaraInt32=2
     7:NiagaraInt32=3
     8:NiagaraInt32=0
     9:NiagaraInt32=1
     10:NiagaraInt32=2
     11:NiagaraInt32=3
     12:NiagaraInt32=0
     13:NiagaraInt32=1
     14:NiagaraInt32=2
     Index:NiagaraInt32=&SpawnGroup()
     15:NiagaraInt32=4
     Num:NiagaraInt32=14
    )) {isSetParameters=true}
   &SolveForcesAndVelocity(Clamp Velocity=false, Limit Acceleration=false)
   &SubUVAnimation(
    Direct Frame Index:NiagaraFloat=$User.Frame
    "Sprite Renderer "=DataInterface() {propertyValues {SpriteRenderer=&self:NE_Balloon_0.NiagaraSpriteRendererProperties_1}}
    SubUV Animation Mode:ENiagara_SubUVLookupModeV2=::NewEnumerator5
   ) {module=SubUVAnimation001}
   &self:WriteDeath_02(Color Array:NiagaraDataInterfaceArrayColor=$System.ColorArray, Position Array:NiagaraDataInterfaceArrayPosition=$System.DeathArray) {module=WriteDeath_02}
  ]
  rendererClasses [&NiagaraSpriteRendererProperties]
  renderers [
   {rendererClass=&NiagaraSpriteRendererProperties}
   {rendererClass=&NiagaraSpriteRendererProperties, rendererIndex=1}
   {rendererClass=&NiagaraSpriteRendererProperties, rendererIndex=2}
   {rendererClass=&NiagaraSpriteRendererProperties, rendererIndex=3}
   {rendererClass=&NiagaraSpriteRendererProperties, rendererIndex=4}
  ]
 }
 NE_Balloon001 {
  emitterUpdateScript [
   &EmitterState(Life Cycle Mode=::NewEnumerator1, Loop Behavior=::NewEnumerator1, Loop Duration Mode=::NewEnumerator1, UseLoopDelay=false)
   &SpawnPerFrame(Spawn:NiagaraBool=$System.DeathSpawn, Use Spawn Probability=false)
  ]
  particleSpawnScript [
   /SetVariables(Particles.NewPosition:NiagaraPosition=&AddVectorToPosition(Position=$System.DeathPosition, Vector=<0 0 13>)) {isSetParameters=true}
   &InitializeParticle(
    Lifetime=&RandomRangeFloat(Minimum=0.12, Maximum=0.16)
    Position Offset Coordinate Space=::Local
    Sprite Rotation Mode=::NewEnumerator2
    Sprite Size Mode=::NewEnumerator3
    Uniform Sprite Size=&RandomRangeFloat(Minimum=35, Maximum=42)
    UsePositionOffset=true
    Write Material Random=true
   )
   &CameraOffset(Camera Offset Amount=5)
  ]
  particleUpdateScript [
   &SubUVAnimation(
    End Frame Range Override=3
    "Sprite Renderer "=DataInterface() {propertyValues {SpriteRenderer=&self:NE_Balloon001_0.NiagaraSpriteRendererProperties_0}}
    SubUV Animation Mode:ENiagara_SubUVLookupModeV2=::NewEnumerator1
    SubUV Curve=DataInterface() {propertyValues {Curve {defaultValue=0, keys [{tangentMode=RCTM_User, time=0, value=0}, {tangentMode=RCTM_User, time=1, value=1}]}}}
    UseEndFrame=false
    UseStartFrame=false
   )
   &ParticleState()
   &SolveForcesAndVelocity(Clamp Velocity=false, Limit Acceleration=false)
  ]
  rendererClasses [&NiagaraSpriteRendererProperties]
  renderers [{rendererClass=&NiagaraSpriteRendererProperties}]
 }
 NE_Spark {
  emitterUpdateScript [
   &EmitterState(
    Enable Distance Culling=true
    Max Distance Response=::KillImmediately
    Module.bApplySpawnCountScaleByDistance=true
    Module.bUseMaxDistance=true
    Module.bUseMinDistance=false
    Scalability Mode=::NewEnumerator1
    Scale Spawn Count=true
    Spawn Count Scale By Distance=DataInterface() {propertyValues {Curve {keys [<at 0.3676 0.9657>,<at 0.9962 0.4982>]}}}
   )
   &UpdateValueStateChange(InputIntParam:NiagaraInt32=&MakeIntFromFloat(Float=$User.BurstSpark, Conversion Mode=::NewEnumerator1))
   &self:TriggerSpawn_04(
    Spawn Count:NiagaraInt32=&self:Merge_Helper_Int_Switch_15(
     0:NiagaraInt32=0
     1:NiagaraInt32=&RandomRangeInt(Minimum=3, Maximum=6)
     2:NiagaraInt32=&RandomRangeInt(Minimum=3, Maximum=6)
     3:NiagaraInt32=&RandomRangeInt(Minimum=3, Maximum=6)
     4:NiagaraInt32=&RandomRangeInt(Minimum=3, Maximum=6)
     5:NiagaraInt32=&RandomRangeInt(Minimum=3, Maximum=6)
     6:NiagaraInt32=&RandomRangeInt(Minimum=3, Maximum=6)
     7:NiagaraInt32=&RandomRangeInt(Minimum=3, Maximum=6)
     8:NiagaraInt32=?
     9:NiagaraInt32=?
     10:NiagaraInt32=?
     11:NiagaraInt32=?
     12:NiagaraInt32=?
     13:NiagaraInt32=?
     14:NiagaraInt32=?
     Index:NiagaraInt32=&MakeIntFromFloat(Float=$User.BurstSpark, Conversion Mode=::NewEnumerator1)
     15:NiagaraInt32=?
     Num:NiagaraInt32=6
    )
    Spawn Group:NiagaraInt32=&MakeIntFromFloat(Float=$User.BurstSpark, Conversion Mode=::NewEnumerator1)
    Trigger:NiagaraBool=$Emitter.UpdateValueStateChange.Changed
   ) {module=TriggerSpawn_04}
  ]
  particleSpawnScript [
   /SetVariables(Particles.IndexIn:NiagaraFloat=$User.BurstSpark) {isSetParameters=true}
   &InitializeParticle(
    Color=&RandomRangeLinearColor(Minimum=<rgba 0.7656 0.3971 0.2074 1>, Maximum=<rgba 0.8281 0.3809 0.2372 1>)
    Lifetime Max=0.45
    Lifetime Min=0.166
    Lifetime Mode=::NewEnumerator1
    Sprite Size Mode=::NewEnumerator1
    Uniform Sprite Size Max=8
    Uniform Sprite Size Min=4
    UsePositionOffset=false
    Write Material Random=true
   )
   &ShapeLocation(
    Hemisphere Distribution=<1 0.2>
    MaskSpawnGroup=false
    Override Seed=false
    Rotation Axis=<1 -0.8861 0.25>
    Rotation Mode=::NewEnumerator1
    Sphere Radius=0.1
    Sphere Surface Distribution=1
    UseOwnerScale=true
   )
   &AddVelocity(
    Module.Use Falloff Distance=false
    Override Seed=false
    Velocity Mode=::NewEnumerator1
    Velocity Speed=&FN_FrontEndFloatModifier(
     Gameplay=&RandomRangeFloat(Minimum=100, Maximum=250)
     Front End=&RandomRangeFloat(Minimum=150, Maximum=400)
    )
   )
   &SkeletalMeshLocation(
    Bone / Socket Index=&MakeIntFromFloat(Float=$Particles.IndexIn, Conversion Mode=::NewEnumerator1)
    Bone Sampling Mode=::NewEnumerator4
    MaskSpawnGroup=false
    Sampled Position Offset=<5 0 15>
    Skeletal Mesh=DataInterface() {propertyValues {
      DefaultMesh=null
      ExcludeBoneName=root
      OptionalComponentClass=null
      PreviewMesh=&SKM_EdibleCable_Gadget
      SkinningMode=SkinOnTheFly
      SoftSourceActor=null
      SourceMode=Default
      UvSetIndex=0
      WholeMeshLOD=-1
      bExcludeBone=false
      bReadDeformedGeometry=true
      bRequireCurrentFrameData=true
      ComponentTags []
      FilteredBones [BalloonBoss,BalloonBoss,BalloonBoss,BalloonBoss,BalloonBoss,BalloonBoss,BalloonBoss]
      FilteredSockets []
      MeshUserParameter {parameter {name=None, typeDefHandle {registeredTypeIndex=1}, varData []}}
      SamplingRegions []
     }}
   )
  ]
  particleUpdateScript [
   &ParticleState()
   &GravityForce(Gravity=<0 0 -666>)
   &Drag(Drag=&UniformRangedFloat(Minimum=2, Maximum=6), Use Linear Drag=true, Use Rotational Drag=false)
   &SolveForcesAndVelocity(Clamp Velocity=false, Limit Acceleration=false)
   &ScaleSpriteSize(Scale Factor=&Vector2DFromCurve(
     Vector2Curve=DataInterface() {propertyValues {
       XCurve {keys [
         {arriveTangent=5.9728, interpMode=RCIM_Cubic, leaveTangent=5.9728, tangentMode=RCTM_User, time=0, value=0.1}
         {arriveTangent=-0.1, interpMode=RCIM_Cubic, leaveTangent=-0.1, tangentMode=RCTM_User, time=0.18, value=0.8}
         {interpMode=RCIM_Cubic, tangentMode=RCTM_User, time=1, value=0}
        ]}
       YCurve {keys [
         {arriveTangent=6.2715, interpMode=RCIM_Cubic, leaveTangent=6.2715, tangentMode=RCTM_User, time=0, value=0.1}
         {arriveTangent=-0.1001, interpMode=RCIM_Cubic, leaveTangent=-0.1001, tangentMode=RCTM_User, time=0.18, value=0.8}
         {interpMode=RCIM_Cubic, tangentMode=RCTM_User, time=0.9988, value=0}
        ]}
      }}
     ScaleCurve=<1 1>
    ))
   &ScaleSpriteSizeBySpeed(Max Scale Factor=<0.5 3>, Min Scale Factor=<0.5 1>, Velocity Threshold=400)
  ]
  rendererClasses [&NiagaraSpriteRendererProperties]
  renderers [{rendererClass=&NiagaraSpriteRendererProperties}]
 }
}
summary {
 systemName=NS_Emote_EdibleCable_Balloon
 emitters [
  {emitterName=NE_Balloon, rendererClasses [&NiagaraSpriteRendererProperties]}
  {emitterName=NE_Balloon001, rendererClasses [&NiagaraSpriteRendererProperties]}
  {emitterName=NE_Spark, rendererClasses [&NiagaraSpriteRendererProperties]}
 ]
 userVariables [
  User.bIsFrontEnd {type=NiagaraBool}
  User.bIsFrontEndPreview {type=NiagaraBool}
  User.BurstIndex {type=NiagaraFloat, defaultValue=0, description=""}
  User.BurstSpark {type=NiagaraFloat, defaultValue=0, description=""}
  User.Frame {type=NiagaraFloat, defaultValue=0, description=""}
 ]
}
