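# Niagara star emote effect for the EdibleCable cosmetic companion: a short golden sparkle burst with glow.
# NE_Glow01 fires small sprites in a cone, braked by heavy drag.
# NE_Glow02 uses spawn-group switches to layer upward-facing glow sprites, a mesh and a light, whose brightness follows time of day and radius shrinks in front-end views.
# NE_Glow_Low is a near-copy without the light, with slightly retuned layers.
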NiagaraSystem /CosmeticCompanions/FX/Builtins/EdibleCable/Emote/NS_Emote_EdibleCable_Star
dependencies {
 usedDataInterfaces [&NiagaraDataInterfaceCurve, &NiagaraDataInterfaceColorCurve]
 usedDynamicInputs []
 usedModules [&SystemState]
}
emitters {
 NE_Glow01 {
  emitterUpdateScript [
   &EmitterState(Life Cycle Mode=::NewEnumerator1, Loop Behavior=::NewEnumerator1, Loop Duration=3, UseLoopDelay=false)
   &SpawnBurst_Instantaneous(Spawn Count=11, Use Spawn Probability=false, UseLoopCountLimit=false)
  ]
  particleSpawnScript [
   &InitializeParticle(
    Color=&FNTimeOfDayColor(Night Color=<rgba 1 1 1 1>, Morning Color=<rgba 1 1 1 1>, Day Color=<rgba 1 1 1 1>, Evening Color=<rgba 1 1 1 1>)
    Lifetime=&RandomRangeFloat(Minimum=1.5, Maximum=2)
    Mass Mode=::NewEnumerator3
    Position Offset Coordinate Space=::World
    Sprite Rotation Mode=::NewEnumerator2
    Sprite Size Mode=::NewEnumerator1
    Uniform Sprite Size Max=18
    Uniform Sprite Size Min=3
    UsePositionOffset=true
    Write Material Random=true
   )
   /SetVariables(Particles.GlowColor=&FNTimeOfDayColor(
     Night Color=<rgba 3 2.518 0.3281 0.4>
     Morning Color=<rgba 3 2.518 0.3281 0.4>
     Day Color=<rgba 3 2.518 0.3281 0.4>
     Evening Color=<rgba 3 2.518 0.3281 0.4>
    )) {isSetParameters=true}
   &ShapeLocation(
    Cone Angle=110
    Cone Length=6
    Cone Radial Angle=180
    Cone Surface Distribution=1
    MaskSpawnGroup=false
    Offset=<-0.5 0 0>
    Offset Mode=::NewEnumerator0
    Override Seed=false
    Rotation Angle Type:ENiagara_AngleInput=::NewEnumerator0
    Rotation Mode=::NewEnumerator2
    Shape Primitive=::NewEnumerator5
    UseOwnerScale=true
    Yaw / Pitch / Roll=<180 0 0>
   ) {module=ShapeLocation001}
   &AddVelocity(
    Module.Use Falloff Distance=false
    Override Seed=false
    Velocity Mode=::NewEnumerator1
    Velocity Speed=&RandomRangeFloat(Minimum=800, Maximum=1200)
   )
   &CameraOffset(Camera Offset Amount=10)
  ]
  particleUpdateScript [
   &ParticleState()
   &SpriteRotationRate(Rotation Rate=&FloatFromCurve(
     FloatCurve=DataInterface() {propertyValues {Curve {defaultValue=1, keys [{tangentMode=RCTM_User, time=0, value=0}, <at 0.1112 0>, <at 0.2001 1>, {tangentMode=RCTM_User, time=1, value=1}]}}}
     Scale Curve=30
    ))
   &ScaleSpriteSize(Uniform Curve Sprite Scale=DataInterface() {propertyValues {Curve {keys [
        {tangentMode=RCTM_User, time=0.0013, value=0.4}
        <at 0.1013 1.3>
        <at 0.1362 0.9>
        <at 0.1849 1>
        <at 0.7481 1>
        {tangentMode=RCTM_User, time=1.0073, value=0}
       ]}}})
   /SetVariables(Particles.Glow_SpriteSize:Vector2f=&Multiply_Vector2D_ByFloat(Float=1.56, Vector2D=$Particles.SpriteSize)) {isSetParameters=true}
   &ScaleColor(
    Scale Alpha=&FloatFromCurve(FloatCurve=DataInterface() {propertyValues {Curve {keys [{arriveTangent=-1.0036, time=0.9, value=1}, <at 1 0 RCIM_Cubic>]}}})
    ScaleA=true
    ScaleRGB=true
   )
   &Drag(Drag=18, Use Linear Drag=true, Use Rotational Drag=false)
   &SolveForcesAndVelocity(Clamp Velocity=false, Limit Acceleration=false)
  ]
  rendererClasses [&NiagaraSpriteRendererProperties]
  renderers [
   {rendererClass=&NiagaraSpriteRendererProperties}
   {rendererClass=&NiagaraSpriteRendererProperties, rendererIndex=1}
  ]
 }
 NE_Glow02 {
  emitterUpdateScript [
   &EmitterState(Life Cycle Mode=::NewEnumerator1, Loop Behavior=::NewEnumerator1, Loop Duration=2, UseLoopDelay=false)
   &SpawnBurst_Instantaneous(Use Spawn Probability=false, UseLoopCountLimit=false)
   &SpawnBurst_Instantaneous(Spawn Group=1, Use Spawn Probability=false, UseLoopCountLimit=false) {module=SpawnBurst_Instantaneous001}
   &SpawnBurst_Instantaneous(Spawn Group=2, Use Spawn Probability=false, UseLoopCountLimit=false) {module=SpawnBurst_Instantaneous002}
   &SpawnBurst_Instantaneous(Spawn Group=3, Use Spawn Probability=false, UseLoopCountLimit=false) {module=SpawnBurst_Instantaneous003}
  ]
  particleSpawnScript [
   /SetVariables(Particles.VisibilityTag=&SpawnGroup()) {isSetParameters=true, module=SetVariables_#1}
   &InitializeParticle(
    Color=&MergeHelper_ColorSwitch(
     0=&FNTimeOfDayColor(
      Night Color=<rgba 3 2.2254 0.6328 0.8>
      Morning Color=<rgba 3 2.2561 0.7266 1>
      Day Color=<rgba 3 2.2561 0.7266 1>
      Evening Color=<rgba 3 2.2561 0.7266 1>
     )
     1=&FNTimeOfDayColor(
      Night Color=<rgba 1 0.8207 0.2969 0.8>
      Morning Color=<rgba 1 0.8207 0.2969 1>
      Day Color=<rgba 1 0.8207 0.2969 1>
      Evening Color=<rgba 1 0.8207 0.2969 1>
     )
     2=&FNTimeOfDayColor(
      Night Color=<rgba 2 1.6413 0.5938 0.8>
      Morning Color=<rgba 2 1.6413 0.5938 1>
      Day Color=<rgba 2 1.6413 0.5938 1>
      Evening Color=<rgba 2 1.6413 0.5938 1>
     )
     3=&FNTimeOfDayColor(
      Night Color=<rgba 2.5 2.0077 0.3581 0.4>
      Morning Color=<rgba 3 2.4397 0.5625 0.6609>
      Day Color=<rgba 3 2.4397 0.5625 0.6609>
      Evening Color=<rgba 3 2.4397 0.5625 0.6609>
     )
     Index=&SpawnGroup()
     Num=::NewEnumerator2
    )
    Lifetime=&MergeHelper_FloatSwitch(0=1.5, 1=1.5, 2=1.5, 3=1.5, Index=&SpawnGroup(), Num=::NewEnumerator2)
    Mass Mode=::NewEnumerator3
    Mesh Scale Mode=::NewEnumerator3
    Mesh Uniform Scale=&MergeHelper_FloatSwitch(0=0, 1=0, 2=0, 3=0.8, Index=&SpawnGroup(), Num=::NewEnumerator2)
    Position Offset=&MergeHelper_VectorSwitch(0=<0 0 15>, 1=<0 0 15>, 2=<0 0 -14>, 3=<0 0 15>, Index=&SpawnGroup(), Num=::NewEnumerator2)
    Sprite Rotation Mode=::NewEnumerator2
    Sprite Size Mode=::NewEnumerator3
    Uniform Sprite Size=&MergeHelper_FloatSwitch(0=270, 1=230, 2=260, 3=0, Index=&SpawnGroup(), Num=::NewEnumerator2)
    UsePositionOffset=true
    Write Material Random=true
   )
   &InitialMeshOrientation(Orientation Axis=<0 0 1>, Orientation Vector=<0 0 1>, Rotation=<0 0.25 0>, Use Orientation Vector=true, Use Rotation Vector=true)
  ]
  particleUpdateScript [
   &ParticleState()
   &ScaleMeshSize(Scale Factor=&VectorFromFloat(Value=&FloatFromCurve(FloatCurve=DataInterface() {propertyValues {Curve {defaultValue=0.7, keys [{tangentMode=RCTM_User, time=0, value=0}, <at 0.05 1>, {tangentMode=RCTM_User, time=1, value=1}]}}})))
   &SpriteRotationRate(Rotation Rate=10)
   &ScaleColor(
    Scale Alpha=&MergeHelper_FloatSwitch(
     0=&FloatFromCurve(
      FloatCurve=DataInterface() {propertyValues {Curve {keys [<at -0.0035 0.192>, <at 0.2 1>, {arriveTangent=-1.0036, time=0.8, value=1}, <at 1.0011 0.002 RCIM_Cubic>]}}}
      Scale Curve=0.3
     )
     1=&FloatFromCurve(FloatCurve=DataInterface() {propertyValues {Curve {keys [<at -0.0035 0.192>, <at 0.1058 1>, {arriveTangent=-1.0036, time=0.835, value=1}, <at 1.0011 0.002 RCIM_Cubic>]}}})
     2=&FloatFromCurve(
      FloatCurve=DataInterface() {propertyValues {Curve {keys [<at -0.0035 0.192>, <at 0.1058 1>, {arriveTangent=-1.0036, time=0.835, value=1}, <at 1.0011 0.002 RCIM_Cubic>]}}}
      Scale Curve=0.15
     )
     3=&FloatFromCurve(FloatCurve=DataInterface() {propertyValues {Curve {defaultValue=1, keys [{tangentMode=RCTM_User, time=0, value=0}, <at 0.1152 1>, <at 0.8165 1>, {tangentMode=RCTM_User, time=1, value=0}]}}})
     Index=&SpawnGroup()
     Num=::NewEnumerator2
    )
    ScaleA=true
    ScaleRGB=true
   )
   &SolveForcesAndVelocity(Clamp Velocity=false, Limit Acceleration=false)
   &SpriteFacingAndAlignment(Sprite Facing=<0 0 1>, WriteAlignment=true, WriteFacing=true) {module=SpriteFacingAndAlignment001}
   &FN_Light_Attributes(
    Color=&ColorFromCurve(ColorCurve=DataInterface() {propertyValues {AlphaCurve {keys []}, BlueCurve {keys [<at 0.098 0.4453>]}, GreenCurve {keys [<at 0.098 5.1012>]}, RedCurve {keys [<at 0.098 5.7906>]}}})
    Create Color Attribute=true
    Create Exponent Attribute=true
    Create Radius Attribute=true
    Exponent:NiagaraFloat=&MakeCustomFloatFromBool(Bool=&Boolean_OrOperation(Bool1=$User.bIsFrontEnd, Bool2=$User.bIsFrontEndPreview), True Float=2, False Float=2)
    Radius=&FloatFromCurve(
     FloatCurve=DataInterface() {propertyValues {Curve {defaultValue=0, keys [{tangentMode=RCTM_User, time=0, value=0}, <at 0.1 1>, {tangentMode=RCTM_User, time=1, value=1}]}}}
     Scale Curve=&MakeCustomFloatFromBool(
      Bool=&Boolean_OrOperation(Bool1=$User.bIsFrontEnd, Bool2=$User.bIsFrontEndPreview)
      True Float=60
      False Float=100
     )
    )
    Scale Color=&FloatFromCurve(
     FloatCurve=DataInterface() {propertyValues {Curve {keys [<at 0 1>,<at 1 0>]}}}
     Scale Curve=&Multiply_Float(
      A=&MakeCustomFloatFromBool(Bool=&Boolean_OrOperation(Bool1=$User.bIsFrontEnd, Bool2=$User.bIsFrontEndPreview), True Float=3, False Float=3)
      B=&FNTimeOfDayScalar(Night Scale=0.3, Morning Scale=0.5, Day Scale=0.8, Evening Scale=0.4)
     )
    )
   )
  ]
  rendererClasses [
   &NiagaraSpriteRendererProperties
   &NiagaraLightRendererProperties
   &NiagaraMeshRendererProperties
  ]
  renderers [
   {rendererClass=&NiagaraSpriteRendererProperties}
   {rendererClass=&NiagaraSpriteRendererProperties, rendererIndex=1}
   {rendererClass=&NiagaraSpriteRendererProperties, rendererIndex=2}
   {rendererClass=&NiagaraLightRendererProperties, rendererIndex=3}
   {rendererClass=&NiagaraMeshRendererProperties, rendererIndex=4}
  ]
 }
 NE_Glow_Low {
  emitterUpdateScript [
   &EmitterState(Life Cycle Mode=::NewEnumerator1, Loop Behavior=::NewEnumerator1, Loop Duration=2, UseLoopDelay=false)
   &SpawnBurst_Instantaneous(Use Spawn Probability=false, UseLoopCountLimit=false)
   &SpawnBurst_Instantaneous(Spawn Group=1, Use Spawn Probability=false, UseLoopCountLimit=false) {module=SpawnBurst_Instantaneous001}
   &SpawnBurst_Instantaneous(Spawn Group=2, Use Spawn Probability=false, UseLoopCountLimit=false) {module=SpawnBurst_Instantaneous002}
   &SpawnBurst_Instantaneous(Spawn Group=3, Use Spawn Probability=false, UseLoopCountLimit=false) {module=SpawnBurst_Instantaneous003}
  ]
  particleSpawnScript [
   /SetVariables(Particles.VisibilityTag=&SpawnGroup()) {isSetParameters=true, module=SetVariables_#1}
   &InitializeParticle(
    Color=&MergeHelper_ColorSwitch(
     0=&FNTimeOfDayColor(
      Night Color=<rgba 3 2.2254 0.6328 0.8>
      Morning Color=<rgba 3 2.2561 0.7266 1>
      Day Color=<rgba 3 2.2561 0.7266 1>
      Evening Color=<rgba 3 2.2561 0.7266 1>
     )
     1=&FNTimeOfDayColor(
      Night Color=<rgba 1 0.8207 0.2969 0.8>
      Morning Color=<rgba 1 0.8207 0.2969 1>
      Day Color=<rgba 1 0.8207 0.2969 1>
      Evening Color=<rgba 1 0.8207 0.2969 1>
     )
     2=&FNTimeOfDayColor(
      Night Color=<rgba 2 1.6413 0.5938 0.8>
      Morning Color=<rgba 2 1.6413 0.5938 1>
      Day Color=<rgba 2 1.6413 0.5938 1>
      Evening Color=<rgba 2 1.6413 0.5938 1>
     )
     3=&FNTimeOfDayColor(
      Night Color=<rgba 2.5 2.0077 0.3581 0.4>
      Morning Color=<rgba 3 2.4397 0.5625 0.6609>
      Day Color=<rgba 3 2.4397 0.5625 0.6609>
      Evening Color=<rgba 3 2.4397 0.5625 0.6609>
     )
     Index=&SpawnGroup()
     Num=::NewEnumerator2
    )
    Lifetime=&MergeHelper_FloatSwitch(0=1.5, 1=1.5, 2=1.5, 3=1.5, Index=&SpawnGroup(), Num=::NewEnumerator2)
    Mass Mode=::NewEnumerator3
    Mesh Scale Mode=::NewEnumerator3
    Mesh Uniform Scale=&MergeHelper_FloatSwitch(0=0, 1=0, 2=0, 3=0.8, Index=&SpawnGroup(), Num=::NewEnumerator2)
    Position Offset=&MergeHelper_VectorSwitch(0=<0 0 15>, 1=<0 0 25>, 2=<0 0 -14>, 3=<0 0 16>, Index=&SpawnGroup(), Num=::NewEnumerator2)
    Sprite Rotation Mode=::NewEnumerator2
    Sprite Size Mode=::NewEnumerator3
    Uniform Sprite Size=&MergeHelper_FloatSwitch(0=270, 1=150, 2=260, 3=0, Index=&SpawnGroup(), Num=::NewEnumerator2)
    UsePositionOffset=true
    Write Material Random=true
   )
   &InitialMeshOrientation(Orientation Axis=<0 0 1>, Orientation Vector=<0 0 1>, Rotation=<0 0.25 0>, Use Orientation Vector=true, Use Rotation Vector=true)
  ]
  particleUpdateScript [
   &ParticleState()
   &ScaleMeshSize(Scale Factor=&VectorFromFloat(Value=&FloatFromCurve(FloatCurve=DataInterface() {propertyValues {Curve {defaultValue=0.7, keys [{tangentMode=RCTM_User, time=0, value=0}, <at 0.05 1>, {tangentMode=RCTM_User, time=1, value=1}]}}})))
   &SpriteRotationRate(Rotation Rate=10)
   &ScaleColor(
    Scale Alpha=&MergeHelper_FloatSwitch(
     0=&FloatFromCurve(
      FloatCurve=DataInterface() {propertyValues {Curve {keys [<at -0.0035 0.192>, <at 0.2 1>, {arriveTangent=-1.0036, time=0.8, value=1}, <at 1.0011 0.002 RCIM_Cubic>]}}}
      Scale Curve=0.3
     )
     1=&FloatFromCurve(
      FloatCurve=DataInterface() {propertyValues {Curve {keys [<at -0.0035 0.192>, <at 0.1058 1>, {arriveTangent=-1.0036, time=0.835, value=1}, <at 1.0011 0.002 RCIM_Cubic>]}}}
      Scale Curve=0.5
     )
     2=&FloatFromCurve(
      FloatCurve=DataInterface() {propertyValues {Curve {keys [<at -0.0035 0.192>, <at 0.1058 1>, {arriveTangent=-1.0036, time=0.835, value=1}, <at 1.0011 0.002 RCIM_Cubic>]}}}
      Scale Curve=0.15
     )
     3=&FloatFromCurve(FloatCurve=DataInterface() {propertyValues {Curve {defaultValue=1, keys [{tangentMode=RCTM_User, time=0, value=0}, <at 0.1152 1>, <at 0.8165 1>, {tangentMode=RCTM_User, time=1, value=0}]}}})
     Index=&SpawnGroup()
     Num=::NewEnumerator2
    )
    ScaleA=true
    ScaleRGB=true
   )
   &SolveForcesAndVelocity(Clamp Velocity=false, Limit Acceleration=false)
   &SpriteFacingAndAlignment(Sprite Facing=<0 0 1>, WriteAlignment=true, WriteFacing=true) {module=SpriteFacingAndAlignment001}
  ]
  rendererClasses [&NiagaraSpriteRendererProperties, &NiagaraMeshRendererProperties]
  renderers [
   {rendererClass=&NiagaraSpriteRendererProperties}
   {rendererClass=&NiagaraSpriteRendererProperties, rendererIndex=1}
   {rendererClass=&NiagaraSpriteRendererProperties, rendererIndex=2}
   {rendererClass=&NiagaraMeshRendererProperties, rendererIndex=3}
  ]
 }
}
summary {
 systemName=NS_Emote_EdibleCable_Star
 emitters [
  {emitterName=NE_Glow01, rendererClasses [&NiagaraSpriteRendererProperties]}
  {
   emitterName=NE_Glow02
   rendererClasses [
    &NiagaraSpriteRendererProperties
    &NiagaraLightRendererProperties
    &NiagaraMeshRendererProperties
   ]
  }
  {emitterName=NE_Glow_Low, rendererClasses [&NiagaraSpriteRendererProperties, &NiagaraMeshRendererProperties]}
 ]
 userVariables [
  User.bIsFrontEnd {type=NiagaraBool}
  User.bIsFrontEndPreview {type=NiagaraBool}
  User.DepthFade {type=NiagaraFloat, defaultValue=100, description=""}
 ]
}
